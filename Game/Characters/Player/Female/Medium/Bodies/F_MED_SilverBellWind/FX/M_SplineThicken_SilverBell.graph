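# Translucent, unlit, two-sided Niagara mesh and ribbon material for spline-thickened streak effects on the SilverBellWind character.
# World position offset sums SplineThicken width, a clamped particle-position pull, Perlin noise displacement and an optional directional push.
# Emissive blends particle-tinted panning detail and highlight textures.
# Opacity multiplies a switchable smoke-texture or gradient shape by noise dissolve, edge and depth fades, with dynamic parameters driving dissolve, width and offsets.
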Material /Game/Characters/Player/Female/Medium/Bodies/F_MED_SilverBellWind/FX/M_SplineThicken_SilverBell
graph {
 properties {
  allowTranslucentCustomDepthWrites=true
  bCastRayTracedShadows=false
  bUsedWithNiagaraMeshParticles=true
  bUsedWithNiagaraRibbons=true
  blendMode=BLEND_Translucent
  pixelDepthOffsetMode=PDOM_Legacy
  shadingModel=MSM_Unlit
  twoSided=true
  usedShadingModels=Unlit
 }
 nodes [
  Add(A=Multiply(A=MaterialFunctionCall() {materialFunction=&DebugTimeSine}.Result) {constB=200}) {constB=0.1}
  n1 = DynamicParameter() {desc=Parameters set by Niagara, parameterIndex=2, paramNames [AddedWPO_Strength, DissolveFromCenter_Power, DissolveFromCenter_Mult, ""]}
  n2 = Multiply(A=ScalarParameter() {defaultValue=1, group=AddedWPO, parameterName=AddedWPO_Strength, sortPriority=2}, B=@n1.AddedWPO_Strength)
  n3 = NamedRerouteUsage()
  Add(A=@n2, B=@n3)
  n4 = DynamicParameter() {desc=Parameters set by Niagara, parameterIndex=2, paramNames [AddedWPO_Strength, DissolveFromCenter_Power, DissolveFromCenter_Mult, ""]}
  Add(A=@n4.DissolveFromCenter_Power) {constB=0.05}
  MaterialFunctionCall!(UVs, Offset, RotationCenter, RotationAngle) {materialFunction=/FortUILibrary/MaterialFunctions/Gradients/MF_UI_ReflectedGradient}
  MaterialFunctionCall() {materialFunction=&SafeNormalize}
  n5 = MaterialFunctionCall() {materialFunction=&DebugTimeSine}
  MaterialFunctionCall!(X range Y range, Gradient Height, ShowGrid?, Thickness, Color, "f(x)"=@n5.Result) {materialFunction=/Engine/Functions/Engine_MaterialFunctions01/Debug/PlotFunctionOnGraph}
  n6 = OneMinus(None=ComponentMask(None=NamedRerouteUsage()) {g=true})
  n7 = ScalarParameter() {defaultValue=0.1, group=AddedWPO, parameterName=AddedWPO_Gradient_PreMult, sortPriority=3}
  n8 = Multiply(
   A=Power(
    Base=Multiply(A=@n6, B=@n7) {constB=5}
    Exp=ScalarParameter() {defaultValue=0.2, group=AddedWPO, parameterName=AddedWPO_GradientPower, sortPriority=4}
   ) {constExponent=5}
   B=ScalarParameter() {defaultValue=1.5, group=AddedWPO, parameterName=AddedWPO_Gradient_PostMult, sortPriority=3}
  )
  n9 = ScalarParameter() {defaultValue=0.5, group=AddedWPO, parameterName=AddedWPO_Subtract}
  n10 = Power(
   Base=Sine(None=Multiply(A=@n6) {constB=0.5})
   Exp=ScalarParameter() {defaultValue=1, group=AddedWPO, parameterName=AddedWPO_DualGradient_Power, sortPriority=6}
  )
  n11 = ScalarParameter() {defaultValue=1, group=AddedWPO, parameterName=AddedWPO_DualGradient_Mult, sortPriority=7}
  n12 = Saturate(None=LinearInterpolate(
    A=Subtract(A=@n8, B=@n9)
    B=Multiply(A=@n10, B=@n11)
    Alpha=ScalarParameter() {group=AddedWPO, parameterName=AddedWPO_UseDualGradient, sortPriority=5}
   ))
  MaterialFunctionCall("f(x)"=@n12) {materialFunction=&PlotFunctionOnGraph}
  n13 = ParticleRandom()
  n14 = VectorParameter() {group=Emissive, parameterName=Emissive Detail UV, sortPriority=1, defaultValue=<rgba 0.5 0.5 0 0.2>}
  n15 = ScalarParameter() {defaultValue=1, group=Shared UV, parameterName=Global Scale U}
  n16 = ScalarParameter() {defaultValue=1, group=Shared UV, parameterName=Global Scale V}
  n17 = Multiply(A=NamedRerouteUsage(), B=Multiply(A=AppendVector(A=@n14.R, B=@n14.R), B=AppendVector(A=@n15, B=@n16)))
  n18 = AppendVector(A=@n14.B, B=@n14.B)
  n19 = Add(A=Add(A=AppendVector(A=@n13, B=@n13), B=Panner(Coordinate=@n17, Speed=@n18)))
  n20 = Multiply(
   A=Desaturation(
    None=TextureSampleParameter2D(UVs=@n19) {group=Emissive, parameterName=Detail, texture=&T_Tile_RainbowColor}.RGB
    Fraction=ScalarParameter() {group=Emissive, parameterName=DesaturateDetail}
   )
   B=ScalarParameter() {defaultValue=1, group=Emissive, parameterName=Emissive Detail Intensity}
  )
  n21 = Multiply(A=@n20, B=ParticleColor().RGB)
  n22 = ScalarParameter() {parameterName=MultParticleColorWithDetail}
  n23 = NamedRerouteUsage()
  n24 = VectorParameter() {group=Emissive, parameterName=Emissive HighLight UV, sortPriority=1, defaultValue=<rgba 0.1 0.25 0.25 0.25>}
  n25 = Multiply(
   A=AppendVector(A=@n24.R, B=@n24.R)
   B=AppendVector(
    A=ScalarParameter() {defaultValue=1, group=Shared UV, parameterName=Global Scale U}
    B=ScalarParameter() {defaultValue=1, group=Shared UV, parameterName=Global Scale V}
   )
  )
  n26 = Add(A=AppendVector(A=ParticleRandom(), B=Constant()), B=Panner(Coordinate=Multiply(A=@n23, B=@n25), Speed=AppendVector(A=@n24.B, B=@n24.B)))
  n27 = DynamicParameter() {defaultValue=<rgba 0 1 1 1>, paramNames [Random Offset, Spline Width Mult, Dissolve, Dissolve Edge Intensity]}
  n28 = LinearInterpolate(B=AppendVector(A=@n27.Random Offset, B=Constant()), Alpha=ScalarParameter() {parameterName=Use Dynamic Parameter to Randomize X UV}) {constAlpha=0}
  n29 = LinearInterpolate(
   A=ScalarParameter() {defaultValue=1, group=Emissive, parameterName=Emissive HighLight Min}
   B=ScalarParameter() {defaultValue=4, group=Emissive, parameterName=Emissive HighLight Max}
   Alpha=TextureSampleParameter2D(UVs=Add(A=@n26, B=@n28) {constB=0}) {group=Emissive, parameterName=Highlight, texture=&T_ShieldCaustics}.R
  )
  n30 = ParticleColor()
  MaterialFunctionCall(Emissive=Add(A=LinearInterpolate(A=@n20, B=@n21, Alpha=@n22), B=Multiply(A=@n29, B=@n30.RGB))) {materialFunction=&MF_FX_EmissiveAdjustments}
  Multiply(A=@n2, B=@n3)
  Multiply(A=@n5.Result) {constB=2}
  Multiply(A=MaterialFunctionCall() {materialFunction=&DebugTimeSine}.Result)
  n31 = TextureCoordinate()
  NamedRerouteDeclaration(None=OneMinus(None=MaterialFunctionCall(XY=@n31) {materialFunction=&Swizzle}.YX)) {name=SwizzledUVs, nodeColor=<rgba 0 1 0.8359 1>}
  n32 = ParticleRandom()
  n33 = VectorParameter() {group=Distortion, parameterName=Distortion UV, sortPriority=1, defaultValue=<rgba 1 1 0 0>}
  n34 = ScalarParameter() {defaultValue=1, group=Shared UV, parameterName=Global Scale U}
  n35 = ScalarParameter() {defaultValue=1, group=Shared UV, parameterName=Global Scale V}
  n36 = Multiply(A=TextureCoordinate(), B=Multiply(A=AppendVector(A=@n33.R, B=@n33.R), B=AppendVector(A=@n34, B=@n35)))
  n37 = AppendVector(A=@n33.B, B=@n33.B)
  n38 = TextureSampleParameter2D(UVs=Add(A=AppendVector(A=@n32, B=@n32), B=Panner(Coordinate=@n36, Speed=@n37))) {
   group=Distortion
   parameterName="Distortion  Tex"
   samplerType=SAMPLERTYPE_Normal
   sortPriority=0
   texture=&T_Weap_Distort_N_02
  }
  NamedRerouteDeclaration(None=Multiply(
    A=ComponentMask(None=@n38.RGB) {g=true, r=true}
    B=ScalarParameter() {defaultValue=0.2, group=Distortion, parameterName=Distortion Intensity, sortPriority=2}
   )) {name=Distortion, nodeColor=<rgba 1 0 0.375 1>}
  n39 = Constant()
  n40 = DynamicParameter() {defaultValue=<rgba 0 1 1 1>, paramNames [Random Offset, Spline Width Mult, Dissolve, Dissolve Edge Intensity]}
  n41 = NamedRerouteUsage()
  n42 = VectorParameter() {
   group=WPO
   parameterName=WPO Depth UV
   sortPriority=6
   channelNames {a=WPO Depth V Pan, b=WPO Depth U Pan, g=WPO Depth V, r=WPO Depth U}
   defaultValue=<rgba 0 0.5 0 0.25>
  }
  n43 = Multiply(
   A=AppendVector(A=@n42.WPO Depth U, B=@n42.WPO Depth U)
   B=AppendVector(
    A=ScalarParameter() {defaultValue=1, group=Shared UV, parameterName=Global Scale U}
    B=ScalarParameter() {defaultValue=1, group=Shared UV, parameterName=Global Scale V}
   )
  )
  n44 = Add(
   A=AppendVector(A=@n40.Random Offset, B=@n40.Random Offset)
   B=Panner(Coordinate=Multiply(A=@n41, B=@n43), Speed=AppendVector(A=@n42.WPO Depth U Pan, B=@n42.WPO Depth U Pan))
  )
  n45 = Multiply(
   A=LinearInterpolate(Alpha=TextureSample(UVs=@n44) {samplerType=SAMPLERTYPE_LinearColor, texture=/Game/Global/Textures/Default/Tileable_Patterns/T_PerlinNoise_Uncompressed}.R) {constB=-1}
   B=ScalarParameter() {defaultValue=7, group=WPO, parameterName=WPO Depth Intensity}
  ) {constB=6}
  n46 = ComponentMask(None=NamedRerouteUsage()) {g=true}
  n47 = Multiply(A=Power(Base=OneMinus(None=@n46), Exp=Constant() {r=2}) {constExponent=0.5}, B=Power(Base=@n46, Exp=Constant() {r=1}) {constExponent=0.5})
  n48 = Constant() {r=6}
  n49 = DynamicParameter() {defaultValue=<rgba 0 1 1 1>, paramNames [Random Offset, Spline Width Mult, Dissolve, Dissolve Edge Intensity]}
  n50 = AppendVector(A=@n49.Random Offset, B=@n49.Random Offset)
  n51 = VectorParameter() {group=WPO, parameterName=WPO Noise UV, sortPriority=6, defaultValue=<rgba 0 0.65 0 0.4>}
  n52 = AppendVector(A=@n51.R, B=@n51.R)
  n53 = AppendVector(
   A=ScalarParameter() {defaultValue=1, group=Shared UV, parameterName=Global Scale U}
   B=ScalarParameter() {defaultValue=1, group=Shared UV, parameterName=Global Scale V}
  )
  n54 = NamedRerouteUsage()
  n55 = ComponentMask(None=@n54) {r=true}
  n56 = Power(Base=ComponentMask(None=@n54) {g=true}) {constExponent=1}
  n57 = Panner(Coordinate=Multiply(A=AppendVector(A=@n55, B=@n56), B=Multiply(A=@n52, B=@n53)), Speed=AppendVector(A=@n51.B, B=@n51.B))
  n58 = LinearInterpolate(Alpha=TextureSample(UVs=Add(A=@n50, B=@n57)) {samplerType=SAMPLERTYPE_LinearColor, texture=&T_PerlinNoise_Uncompressed}.R) {constA=-1}
  n59 = MaterialFunctionCall(R=@n58, G=Constant(), B=Multiply(A=@n58) {constB=0.5}) {materialFunction=&AppendMany}
  n60 = DynamicParameter() {parameterIndex=1, defaultValue=<rgba 0 10 1 1>, paramNames [Random Offset, Noise Intensty, Dissolve, Dissolve Edge Intensity]}
  n61 = LinearInterpolate(
   A=ScalarParameter() {defaultValue=3, group=WPO, parameterName=WPO Noise Intensity}
   B=@n60.Noise Intensty
   Alpha=ScalarParameter() {parameterName=UseDynNoiseIntensity}
  )
  n62 = ComponentMask(None=NamedRerouteUsage()) {g=true}
  n63 = Constant() {r=0.9}
  n64 = Multiply(
   A=Power(Base=OneMinus(None=@n62), Exp=Constant() {r=1.8}) {constExponent=0.5}
   B=LinearInterpolate(B=Power(Base=@n62, Exp=@n63) {constExponent=0.5}, Alpha=ScalarParameter() {group=SplineThicken, parameterName=WPO Top Mask}) {constA=1}
  )
  n65 = CurveAtlasRowParameter(CurveTime=ComponentMask(None=NamedRerouteUsage()) {g=true}) {
   atlas=/Game/Athena/LinearColorCurves/A_Atlas_Turtleneck_Glider
   curve=/Game/Accessories/FORT_Backpacks/Backpack_M_MED_Calavera/FX/Curve_Backpack_Calavera_01
   defaultValue=39
   group=SplineThicken
   parameterName=Curve Shape
  }
  n66 = ScalarParameter() {group=WPO, parameterName=Lerp WPO Curve}
  n67 = Add(
   A=Multiply(
    A=MaterialFunctionCall(R=@n39, G=@n45, B=@n39) {materialFunction=&AppendMany}.RGB
    B=Multiply(A=@n47, B=@n48) {constB=4}
   )
   B=Multiply(A=Multiply(A=@n59.RGB, B=@n61), B=LinearInterpolate(A=@n64, B=@n65.G, Alpha=@n66))
  )
  NamedRerouteDeclaration(None=@n67) {name=WPO, nodeColor=<rgba 0 1 1 1>}
  n68 = NamedRerouteUsage()
  n69 = Sine(None=Multiply(A=ComponentMask(None=@n68) {g=true}) {constB=0.5})
  OneMinus(None=@n69)
  n70 = ComponentMask(None=NamedRerouteUsage()) {g=true}
  OneMinus(None=@n70)
  n71 = DynamicParameter() {
   desc=Parameters set by Niagara
   parameterIndex=3
   defaultValue=<rgba 0 1 1 1>
   paramNames [ShapeTypeShift, Spline Width Mult, Dissolve, Dissolve Edge Intensity]
  }
  n72 = MaterialFunctionCall(Float3=Subtract(
    A=MaterialFunctionCall(X=@n71.ShapeTypeShift, Y=@n71.ShapeTypeShift, Z=@n71.ShapeTypeShift) {materialFunction=&MakeFloat3}.Result
    B=ParticlePositionWS()
   )) {materialFunction=&BreakOutFloat3Components}
  n73 = ScalarParameter() {defaultValue=100, group=Interpolate, parameterName=Interpolate Clamp XY}
  n74 = ScalarParameter() {defaultValue=20, group=Interpolate, parameterName=Interpolate Clamp Z}
  n75 = AppendVector(
   A=Clamp(None=AppendVector(A=@n72.R, B=@n72.R), Min=Multiply(A=@n73) {constB=-1}, Max=@n73) {maxDefault=1000, minDefault=-1000}
   B=Clamp(None=@n72.B, Min=Multiply(A=@n74) {constB=-1}, Max=@n74) {minDefault=-1}
  )
  n76 = OneMinus(None=NamedRerouteUsage())
  n77 = Multiply(
   A=@n71.Dissolve Edge Intensity
   B=Power(Base=ComponentMask(None=@n76) {g=true}, Exp=ScalarParameter() {defaultValue=1.7, group=Interpolate, parameterName=Interpolate Power})
  )
  n78 = ComponentMask(None=TextureCoordinate()) {r=true}
  n79 = Multiply(
   A=LinearInterpolate(A=ScalarParameter() {defaultValue=0.5, parameterName=A}, Alpha=OneMinus(None=@n78)) {constA=0.2}
   B=ScalarParameter() {defaultValue=3, parameterName=SplineThickenWidth}
  )
  n80 = ScalarParameter() {defaultValue=0.1, parameterName=SplineWidthSubtract}
  n81 = MaterialFunctionCall(
   WidthBase=Subtract(A=@n79, B=Subtract(A=@n79, B=@n80) {constB=0.1}) {constB=3.9}
   WidthTip=@n79
   Expand U or V UV Channel=StaticBoolParameter() {parameterName="Expand U or V UV Channel (SB)"}
  ) {materialFunction=&SplineThicken}
  n82 = DynamicParameter() {defaultValue=<rgba 0 1 1 1>, paramNames [Random Offset, Spline Width Mult, Dissolve, Dissolve Edge Intensity]}
  n83 = Multiply(A=ScalarParameter() {defaultValue=1, group=SplineThicken, parameterName=Spline Width Mult}, B=@n82.Spline Width Mult)
  n84 = Add(
   A=StaticSwitchParameter(True=LinearInterpolate(B=@n75, Alpha=@n77), False=Constant()) {defaultValue=true, group=Interpolate, parameterName=Use Interpolate WPO, sortPriority=0}
   B=StaticSwitchParameter(True=Multiply(A=@n81.WorldPositionOffset, B=@n83), False=Constant()) {defaultValue=true, parameterName=UseSplineThicken}
  )
  n85 = Transform(None=@n67) {transformSourceType=TRANSFORMSOURCE_Local}
  n86 = VectorParameter() {group=AddedWPO, parameterName=AddeWPODirection, sortPriority=1, defaultValue=<rgba 1 0 0 1>}
  n87 = Transform(None=Multiply(A=Multiply(A=@n86.RGB, B=@n12), B=@n2)) {transformSourceType=TRANSFORMSOURCE_Local}
  n88 = ScalarParameter() {group=AddedWPO, parameterName=UseAddedWPO, sortPriority=0}
  Reroute(None=Add(A=Add(A=@n84, B=@n85) {constB=0}, B=LinearInterpolate(B=@n87, Alpha=@n88))) {desc=WPO}
  ScalarParameter() {defaultValue=1, group=Dissolve, parameterName=Dissolve Dynamic Test}
  ScalarParameter() {defaultValue=1, group=DissolveFromCenter, parameterName=DissolveFromCenter_Power, sortPriority=1}
  ScalarParameter() {defaultValue=1, group=DissolveFromCenter, parameterName=DissolveFromCenter_Mult, sortPriority=2}
  n89 = Constant()
  n90 = ParticleRandom()
  n91 = VectorParameter() {group=Opacity, parameterName=SmokeTrail UV, sortPriority=1, defaultValue=<rgba 1 1 0 0.2>}
  n92 = ScalarParameter() {defaultValue=1, group=Shared UV, parameterName=Global Scale U}
  n93 = ScalarParameter() {defaultValue=1, group=Shared UV, parameterName=Global Scale V}
  n94 = Multiply(A=NamedRerouteUsage(), B=Multiply(A=AppendVector(A=@n91.R, B=@n91.R), B=AppendVector(A=@n92, B=@n93)))
  n95 = AppendVector(A=@n91.B, B=@n91.B)
  n96 = Add(A=Add(A=AppendVector(A=@n89, B=@n90), B=Panner(Coordinate=@n94, Speed=@n95)), B=NamedRerouteUsage())
  n97 = Constant() {r=0.25}
  n98 = TextureSampleParameter2D(UVs=Add(
    A=MaterialFunctionCall(UVs=@n96, "Rotation Angle (0-1)"=@n97) {materialFunction=&CustomRotator}.Rotated Values
    B=@n28
   ) {constB=0}) {parameterName=SmokeTrailTexture, texture=/Game/Effects/Fort_Effects/Textures/Smoke/T_Smoke_Trail_Soft_Chromatic}
  n99 = Multiply(A=Power(Base=@n70)) {constB=6}
  n100 = Multiply(A=ComponentMask(None=NamedRerouteUsage()) {g=true}) {constB=3}
  n101 = LinearInterpolate(A=Multiply(A=Saturate(None=@n99), B=Saturate(None=@n100)), Alpha=ScalarParameter() {parameterName=BypassFade})
  n102 = OneMinus(None=MaterialFunctionCall() {materialFunction=&LinearGradient}.UGradient)
  n103 = ScalarParameter() {defaultValue=1, parameterName="VerticalFade Power (S)"}
  n104 = Multiply(
   A=OneMinus(None=MaterialFunctionCall() {materialFunction=&LinearGradient}.UGradient)
   B=OneMinus(None=Power(Base=@n102, Exp=@n103))
  )
  n105 = NamedRerouteUsage()
  n106 = Saturate(None=Multiply(
    A=MaterialFunctionCall(UVs=@n105) {materialFunction=&MF_UI_ReflectedGradient}.Result
    B=ScalarParameter() {defaultValue=10, group=Opacity, parameterName="HorizontalFade Power (S)"}
   ))
  n107 = NamedRerouteUsage()
  n108 = Multiply(A=ComponentMask(None=@n107) {r=true}, B=OneMinus(None=ComponentMask(None=@n107) {r=true}))
  n109 = Multiply(
   A=LinearInterpolate(B=Multiply(A=@n108) {constB=2}, Alpha=ScalarParameter() {defaultValue=-0.5, group=Dissolve, parameterName=Lerp Gradient Tip}) {constA=1, constAlpha=1}
   B=ComponentMask(None=NamedRerouteUsage()) {g=true}
  )
  n110 = ParticleRandom()
  n111 = AppendVector(A=@n110, B=@n110)
  n112 = VectorParameter() {group=Dissolve, parameterName=Dissolve Tex UV, sortPriority=6, defaultValue=<rgba 0.1 0.5 0 0.2>}
  n113 = Panner(Coordinate=Multiply(A=NamedRerouteUsage(), B=AppendVector(A=@n112.R, B=@n112.R)), Speed=AppendVector(A=@n112.B, B=@n112.B))
  n114 = Multiply(
   A=@n109
   B=TextureSampleParameter2D(UVs=Add(A=@n111, B=@n113)) {group=Dissolve, parameterName=Dissolve Tex, texture=/Game/Global/Textures/Default/Noise/T_FireNoise_01}.R
  )
  n115 = ScalarParameter() {defaultValue=1, group=Dissolve, parameterName=Lerp Dissolve Tex}
  n116 = ScalarParameter() {defaultValue=1, group=Dissolve, parameterName=Dissolve Edge Hardness, sliderMin=0.5}
  n117 = Multiply(
   A=Power(Base=LinearInterpolate(A=@n109, B=@n114, Alpha=@n115), Exp=ScalarParameter() {defaultValue=1, group=Dissolve, parameterName=Dissolve Power})
   B=@n116
  )
  n118 = MaterialFunctionCall() {materialFunction=&LinearGradient}
  n119 = Add(A=@n69) {constB=-0.15}
  n120 = OneMinus(None=Multiply(A=Power(Base=@n119, Exp=@n4.DissolveFromCenter_Power) {constExponent=0.2}, B=@n4.DissolveFromCenter_Mult))
  n121 = LinearInterpolate(
   A=@n116
   Alpha=LinearInterpolate(
    A=OneMinus(None=@n118.UGradient)
    B=Saturate(None=@n120)
    Alpha=ScalarParameter() {group=DissolveFromCenter, parameterName=UseDissolveFromCenter, sortPriority=0}
   )
  ) {constB=-1}
  n122 = Multiply(
   A=StaticSwitchParameter(True=Multiply(A=@n98.R, B=@n101), False=Multiply(A=@n104, B=@n106)) {parameterName=Smoke}
   B=Saturate(None=Subtract(A=@n117, B=@n121))
  )
  n123 = TextureCoordinate()
  n124 = LinearInterpolate(
   A=SmoothStep(
    Min=ScalarParameter() {parameterName=MinSplineOpacity}
    Max=ScalarParameter() {defaultValue=0.1, parameterName=MaxSplineOpacity}
    Value=ComponentMask(None=@n123) {r=true}
   )
   Alpha=ScalarParameter() {parameterName=BypassFade}
  )
  n125 = Power(
   Base=Subtract(
    A=MaterialFunctionCall() {materialFunction=&LinearGradient}.UGradient
    B=ScalarParameter() {defaultValue=-0.95, group=BaseEdgeFade, parameterName=BaseEdgeFade_Subtract}
   )
   Exp=ScalarParameter() {defaultValue=40, group=BaseEdgeFade, parameterName=BaseEdgeFade_Power}
  )
  n126 = Multiply(A=@n30.A, B=Multiply(A=Multiply(A=@n122, B=@n124), B=Saturate(None=@n125)) {desc=fade leading edge})
  n127 = DepthFade(Opacity=@n126, FadeDistance=ScalarParameter() {defaultValue=0.5, parameterName=FadeDistance}) {fadeDistanceDefault=1}
  ShadingPathSwitch(Default=@n127, Deferred=@n127, Mobile=@n126) {desc=Migrated from FeatureLevelSwitch}
  TextureCoordinate()
  TextureCoordinate()
  TextureCoordinate()
 ]
}
