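# Physics asset for the SilverBellWind medium female player body, set up for two dangling waist strands, left and right.
# Each strand is a three-capsule chain with tight swing and twist limits, rooted in a zero-size MSTR body that joins the pelvis with free angular motion.
# Larger pelvis and thigh capsules provide the body volumes around the strands, with the thighs also angularly free relative to the pelvis.
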PhysicsAsset /Game/Characters/Player/Female/Medium/Bodies/F_MED_SilverBellWind/Meshes/F_MED_SilverBellWind_Physics
bodies {
 dyn_waiststrand_l_01_jnt [{shapeType=Capsule, length=0.505, radius=0.505, shapeName=dyn_waiststrand_l_01_jnt_capsule, center=<1 0 0>, rotation=<rot 90 0 0>}]
 dyn_waiststrand_l_02_jnt [{shapeType=Capsule, length=5, radius=0.505, shapeName=dyn_waiststrand_l_02_jnt_capsule, center=<3.5 0 0>, rotation=<rot 90 0 0>}]
 dyn_waiststrand_l_03_jnt [{shapeType=Capsule, length=3, radius=0.505, shapeName=dyn_waiststrand_l_03_jnt_capsule, center=<2.25 0 0>, rotation=<rot 90 0 0>}]
 dyn_waiststrand_l_MSTR [{shapeType=Capsule, length=0, radius=0, shapeName=dyn_waiststrand_l_MSTR_capsule}]
 dyn_waiststrand_r_01_jnt [{shapeType=Capsule, length=0.505, radius=0.505, shapeName=dyn_waiststrand_r_01_jnt_capsule, center=<1 0 0>, rotation=<rot 90 0 0>}]
 dyn_waiststrand_r_02_jnt [{shapeType=Capsule, length=5, radius=0.505, shapeName=dyn_waiststrand_r_02_jnt_capsule, center=<3.5 0 0>, rotation=<rot 90 0 0>}]
 dyn_waiststrand_r_03_jnt [{shapeType=Capsule, length=3, radius=0.505, shapeName=dyn_waiststrand_r_03_jnt_capsule, center=<2.25 0 0>, rotation=<rot 90 0 0>}]
 dyn_waiststrand_r_MSTR [{shapeType=Capsule, length=0, radius=0, shapeName=dyn_waiststrand_r_MSTR_capsule}]
 pelvis [{shapeType=Capsule, length=10, radius=10.5, shapeName=pelvis_capsule, center=<-4 -2 0>}]
 thigh_l [{shapeType=Capsule, length=20, radius=9.5, shapeName=thigh_l_capsule, center=<-15 -1 0>, rotation=<rot 90 0 5>}]
 thigh_r [{shapeType=Capsule, length=20, radius=9.5, shapeName=thigh_r_capsule, center=<15 -1 0>, rotation=<rot 90 0 5>}]
}
constraints [
 {
  bone1Name=thigh_l
  bone2Name=pelvis
  swing1LimitDegrees=45
  swing1Motion=Free
  swing2LimitDegrees=45
  swing2Motion=Free
  twistLimitDegrees=45
  twistMotion=Free
 }
 {
  bone1Name=thigh_r
  bone2Name=pelvis
  swing1LimitDegrees=45
  swing1Motion=Free
  swing2LimitDegrees=45
  swing2Motion=Free
  twistLimitDegrees=45
  twistMotion=Free
 }
 {bone1Name=dyn_waiststrand_r_01_jnt, bone2Name=dyn_waiststrand_r_MSTR, swing1LimitDegrees=15, swing2LimitDegrees=15, twistLimitDegrees=2}
 {bone1Name=dyn_waiststrand_r_02_jnt, bone2Name=dyn_waiststrand_r_01_jnt, swing1LimitDegrees=15, swing2LimitDegrees=15, twistLimitDegrees=2}
 {bone1Name=dyn_waiststrand_r_03_jnt, bone2Name=dyn_waiststrand_r_02_jnt, swing1LimitDegrees=15, swing2LimitDegrees=15, twistLimitDegrees=2}
 {
  bone1Name=dyn_waiststrand_r_MSTR
  bone2Name=pelvis
  swing1LimitDegrees=45
  swing1Motion=Free
  swing2LimitDegrees=45
  swing2Motion=Free
  twistLimitDegrees=45
  twistMotion=Free
 }
 {bone1Name=dyn_waiststrand_l_03_jnt, bone2Name=dyn_waiststrand_l_02_jnt, swing1LimitDegrees=15, swing2LimitDegrees=15, twistLimitDegrees=2}
 {bone1Name=dyn_waiststrand_l_02_jnt, bone2Name=dyn_waiststrand_l_01_jnt, swing1LimitDegrees=15, swing2LimitDegrees=15, twistLimitDegrees=2}
 {bone1Name=dyn_waiststrand_l_01_jnt, bone2Name=dyn_waiststrand_l_MSTR, swing1LimitDegrees=15, swing2LimitDegrees=15, twistLimitDegrees=2}
 {
  bone1Name=dyn_waiststrand_l_MSTR
  bone2Name=pelvis
  swing1LimitDegrees=45
  swing1Motion=Free
  swing2LimitDegrees=45
  swing2Motion=Free
  twistLimitDegrees=45
  twistMotion=Free
 }
]
