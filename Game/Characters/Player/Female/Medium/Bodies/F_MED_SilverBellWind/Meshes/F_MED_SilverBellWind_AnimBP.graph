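# Post-process style cosmetic AnimBP for the F_MED_SilverBellWind player body, adding secondary motion to clothing and accessories over the attached parent's copied pose.
# A Control Rig feeds chained AnimDynamics for sleeves, shorts, tassels, jacket, wristband and waist pack, then a pelvis-space RigidBody simulation.
# Update logic fades sleeves and shorts out during emotes and grinding, and shorts when crouching.
# Crouching also drives the rig and pitches the waist strands afterward.
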AnimBlueprint /Game/Characters/Player/Female/Medium/Bodies/F_MED_SilverBellWind/Meshes/F_MED_SilverBellWind_AnimBP
target_skeleton=&/F_MED_SilverBellWind_Skeleton
graphs [
 AnimGraph {nodes [
   n1 = Control Rig!(
    Source=Copy Pose From Mesh() {bUseAttachedParent=true}
    Alpha
    Is Crouching Alpha:double=Get Is Crouching Alpha() -> (Is Crouching Alpha:double)
   ) -> (Pose) {
    alphaBoolBlend {blendOption=Linear}
    controlRigAssetReference {blueprintRigClass=&/F_MED_SilverBellWind_CtrlRig.F_MED_SilverBellWind_CtrlRig_C}
   }
   n2 = Get Sleeve Alpha() -> (Sleeve Alpha:double)
   n3 = Anim Dynamics - Bone: armsleeve_l!(ComponentPose=Local To Component(LocalPose=@n1), Alpha=@n2, GravityOverride:Vector="0,0,0", ExternalForce="0,0,0") -> (Pose) {
    angularDampingOverride=1
    angularSpringConstant=800
    bAngularSpring=true
    bOverrideAngularDamping=true
    bUseGravityOverride=true
    simulationSpace=RootRelative
    alphaBoolBlend {blendOption=Linear}
    boundBone {boneName=armsleeve_l}
    componentAppliedLinearAccClamp=<6000 6000 6000>
    componentLinearAccScale=<0.3 0.3 0.3>
    componentLinearVelScale=<0.3 0.3 0.3>
    physicsBodyDefinitions [{
      boundBone {boneName=armsleeve_l}
      boxExtents=<40 40 40>
      constraintSetup {angularLimitsMax=<1 2.5 7>, angularLimitsMin=<-1 -5 -7>, angularTarget=<1 0 0>}
      localJointOffset=<8 0 0>
     }]
   }
   n4 = Anim Dynamics - Bone: armsleeve_r!(ComponentPose=@n3, Alpha=@n2, GravityOverride:Vector="0,0,0", ExternalForce="0,0,0") -> (Pose) {
    angularDampingOverride=1
    angularSpringConstant=800
    bAngularSpring=true
    bOverrideAngularDamping=true
    bUseGravityOverride=true
    simulationSpace=RootRelative
    alphaBoolBlend {blendOption=Linear}
    boundBone {boneName=armsleeve_r}
    componentAppliedLinearAccClamp=<6000 6000 6000>
    componentLinearAccScale=<0.3 0.3 0.3>
    componentLinearVelScale=<0.3 0.3 0.3>
    physicsBodyDefinitions [{
      boundBone {boneName=armsleeve_r}
      boxExtents=<40 40 40>
      constraintSetup {angularLimitsMax=<1 2.5 7>, angularLimitsMin=<-1 -5 -7>, angularTarget=<1 0 0>}
      localJointOffset=<8 0 0>
     }]
   }
   n5 = Get Shorts Alpha() -> (Shorts Alpha:double)
   n6 = Anim Dynamics - Bone: legshort_l!(ComponentPose=@n4, Alpha=@n5, GravityOverride:Vector="0,0,0", ExternalForce="0,0,0") -> (Pose) {
    angularDampingOverride=1
    angularSpringConstant=200
    bAngularSpring=true
    bOverrideAngularDamping=true
    bUseGravityOverride=true
    simulationSpace=RootRelative
    alphaBoolBlend {blendOption=Linear}
    boundBone {boneName=legshort_l}
    componentAppliedLinearAccClamp=<6000 6000 6000>
    componentLinearAccScale=<0.3 0.3 0.3>
    componentLinearVelScale=<0.3 0.3 0.3>
    physicsBodyDefinitions [{
      boundBone {boneName=legshort_l}
      boxExtents=<20 20 20>
      constraintSetup {angularLimitsMax=<1 3 4>, angularLimitsMin=<-1 -3 -4>, angularTarget=<1 0 0>}
      localJointOffset=<6 0 0>
     }]
   }
   n7 = Anim Dynamics - Bone: legshort_r!(ComponentPose=@n6, Alpha=@n5, GravityOverride:Vector="0,0,0", ExternalForce="0,0,0") -> (Pose) {
    angularDampingOverride=1
    angularSpringConstant=200
    bAngularSpring=true
    bOverrideAngularDamping=true
    bUseGravityOverride=true
    simulationSpace=RootRelative
    alphaBoolBlend {blendOption=Linear}
    boundBone {boneName=legshort_r}
    componentAppliedLinearAccClamp=<6000 6000 6000>
    componentLinearAccScale=<0.3 0.3 0.3>
    componentLinearVelScale=<0.3 0.3 0.3>
    physicsBodyDefinitions [{
      boundBone {boneName=legshort_r}
      boxExtents=<20 20 20>
      constraintSetup {angularLimitsMax=<1 3 4>, angularLimitsMin=<-1 -3 -4>, angularTarget=<1 0 0>}
      localJointOffset=<6 0 0>
     }]
   }
   n8 = Anim Dynamics - Bone: dyn_ArmTassle_l_01_jnt!(ComponentPose=@n7, Alpha=1, GravityOverride:Vector="0,0,0", ExternalForce="0,0,0") -> (Pose) {
    angularDampingOverride=1
    angularSpringConstant=200
    bAngularSpring=true
    bOverrideAngularDamping=true
    bUseGravityOverride=true
    simulationSpace=RootRelative
    alphaBoolBlend {blendOption=Linear}
    boundBone {boneName=dyn_ArmTassle_l_01_jnt}
    componentAppliedLinearAccClamp=<6000 6000 6000>
    componentLinearAccScale=<0.3 0.3 0.3>
    componentLinearVelScale=<0.3 0.3 0.3>
    physicsBodyDefinitions [{
      boundBone {boneName=dyn_ArmTassle_l_01_jnt}
      boxExtents=<10 10 10>
      constraintSetup {angularLimitsMax=<1 6 30>, angularLimitsMin=<-1 -30 -30>, angularTarget=<1 0 0>}
      localJointOffset=<2 0 0>
     }]
   }
   n9 = Anim Dynamics - Bone: dyn_ArmTassle_l_02_jnt!(ComponentPose=@n8, Alpha=1, GravityOverride:Vector="0,0,0", ExternalForce="0,0,0") -> (Pose) {
    angularDampingOverride=1
    angularSpringConstant=200
    bAngularSpring=true
    bOverrideAngularDamping=true
    bUseGravityOverride=true
    simulationSpace=RootRelative
    alphaBoolBlend {blendOption=Linear}
    boundBone {boneName=dyn_ArmTassle_l_02_jnt}
    componentAppliedLinearAccClamp=<6000 6000 6000>
    componentLinearAccScale=<0.3 0.3 0.3>
    componentLinearVelScale=<0.3 0.3 0.3>
    physicsBodyDefinitions [{
      boundBone {boneName=dyn_ArmTassle_l_02_jnt}
      boxExtents=<10 10 10>
      constraintSetup {angularLimitsMax=<1 6 30>, angularLimitsMin=<-1 -30 -30>, angularTarget=<1 0 0>}
      localJointOffset=<2 0 0>
     }]
   }
   n10 = Anim Dynamics - Bone: dyn_ArmTassle_l_03_jnt!(ComponentPose=@n9, Alpha=1, GravityOverride:Vector="0,0,0", ExternalForce="0,0,0") -> (Pose) {
    angularDampingOverride=1
    angularSpringConstant=200
    bAngularSpring=true
    bOverrideAngularDamping=true
    bUseGravityOverride=true
    simulationSpace=RootRelative
    alphaBoolBlend {blendOption=Linear}
    boundBone {boneName=dyn_ArmTassle_l_03_jnt}
    componentAppliedLinearAccClamp=<6000 6000 6000>
    componentLinearAccScale=<0.3 0.3 0.3>
    componentLinearVelScale=<0.3 0.3 0.3>
    physicsBodyDefinitions [{
      boundBone {boneName=dyn_ArmTassle_l_03_jnt}
      boxExtents=<10 10 10>
      constraintSetup {angularLimitsMax=<1 6 30>, angularLimitsMin=<-1 -30 -30>, angularTarget=<1 0 0>}
      localJointOffset=<2 0 0>
     }]
   }
   n11 = Anim Dynamics - Bone: dyn_ArmTassle_r_01_jnt!(ComponentPose=@n10, Alpha=1, GravityOverride:Vector="0,0,0", ExternalForce="0,0,0") -> (Pose) {
    angularDampingOverride=1
    angularSpringConstant=200
    bAngularSpring=true
    bOverrideAngularDamping=true
    bUseGravityOverride=true
    simulationSpace=RootRelative
    alphaBoolBlend {blendOption=Linear}
    boundBone {boneName=dyn_ArmTassle_r_01_jnt}
    componentAppliedLinearAccClamp=<6000 6000 6000>
    componentLinearAccScale=<0.3 0.3 0.3>
    componentLinearVelScale=<0.3 0.3 0.3>
    physicsBodyDefinitions [{
      boundBone {boneName=dyn_ArmTassle_r_01_jnt}
      boxExtents=<10 10 10>
      constraintSetup {angularLimitsMax=<1 6 30>, angularLimitsMin=<-1 -30 -30>, angularTarget=<1 0 0>}
      localJointOffset=<2 0 0>
     }]
   }
   n12 = Anim Dynamics - Bone: dyn_ArmTassle_r_02_jnt!(ComponentPose=@n11, Alpha=1, GravityOverride:Vector="0,0,0", ExternalForce="0,0,0") -> (Pose) {
    angularDampingOverride=1
    angularSpringConstant=200
    bAngularSpring=true
    bOverrideAngularDamping=true
    bUseGravityOverride=true
    simulationSpace=RootRelative
    alphaBoolBlend {blendOption=Linear}
    boundBone {boneName=dyn_ArmTassle_r_02_jnt}
    componentAppliedLinearAccClamp=<6000 6000 6000>
    componentLinearAccScale=<0.3 0.3 0.3>
    componentLinearVelScale=<0.3 0.3 0.3>
    physicsBodyDefinitions [{
      boundBone {boneName=dyn_ArmTassle_r_02_jnt}
      boxExtents=<10 10 10>
      constraintSetup {angularLimitsMax=<1 6 30>, angularLimitsMin=<-1 -30 -30>, angularTarget=<1 0 0>}
      localJointOffset=<2 0 0>
     }]
   }
   n13 = Anim Dynamics - Bone: dyn_ArmTassle_r_03_jnt!(ComponentPose=@n12, Alpha=1, GravityOverride:Vector="0,0,0", ExternalForce="0,0,0") -> (Pose) {
    angularDampingOverride=1
    angularSpringConstant=200
    bAngularSpring=true
    bOverrideAngularDamping=true
    bUseGravityOverride=true
    simulationSpace=RootRelative
    alphaBoolBlend {blendOption=Linear}
    boundBone {boneName=dyn_ArmTassle_r_03_jnt}
    componentAppliedLinearAccClamp=<6000 6000 6000>
    componentLinearAccScale=<0.3 0.3 0.3>
    componentLinearVelScale=<0.3 0.3 0.3>
    physicsBodyDefinitions [{
      boundBone {boneName=dyn_ArmTassle_r_03_jnt}
      boxExtents=<10 10 10>
      constraintSetup {angularLimitsMax=<1 6 30>, angularLimitsMin=<-1 -30 -30>, angularTarget=<1 0 0>}
      localJointOffset=<2 0 0>
     }]
   }
   n14 = Anim Dynamics - Bone: dyn_Leg_Tassle_l_01_jnt!(ComponentPose=@n13, Alpha=1, GravityOverride:Vector="0,0,0", ExternalForce="0,0,0") -> (Pose) {
    angularDampingOverride=1
    angularSpringConstant=200
    bAngularSpring=true
    bOverrideAngularDamping=true
    bUseGravityOverride=true
    simulationSpace=RootRelative
    alphaBoolBlend {blendOption=Linear}
    boundBone {boneName=dyn_Leg_Tassle_l_01_jnt}
    componentAppliedLinearAccClamp=<6000 6000 6000>
    componentLinearAccScale=<0.3 0.3 0.3>
    componentLinearVelScale=<0.3 0.3 0.3>
    physicsBodyDefinitions [{
      boundBone {boneName=dyn_Leg_Tassle_l_01_jnt}
      boxExtents=<10 10 10>
      constraintSetup {angularLimitsMax=<1 6 30>, angularLimitsMin=<-1 -30 -30>, angularTarget=<1 0 0>}
      localJointOffset=<2 0 0>
     }]
   }
   n15 = Anim Dynamics - Bone: dyn_Leg_Tassle_l_02_jnt!(ComponentPose=@n14, Alpha=1, GravityOverride:Vector="0,0,0", ExternalForce="0,0,0") -> (Pose) {
    angularDampingOverride=1
    angularSpringConstant=200
    bAngularSpring=true
    bOverrideAngularDamping=true
    bUseGravityOverride=true
    simulationSpace=RootRelative
    alphaBoolBlend {blendOption=Linear}
    boundBone {boneName=dyn_Leg_Tassle_l_02_jnt}
    componentAppliedLinearAccClamp=<6000 6000 6000>
    componentLinearAccScale=<0.3 0.3 0.3>
    componentLinearVelScale=<0.3 0.3 0.3>
    physicsBodyDefinitions [{
      boundBone {boneName=dyn_Leg_Tassle_l_02_jnt}
      boxExtents=<10 10 10>
      constraintSetup {angularLimitsMax=<1 6 30>, angularLimitsMin=<-1 -30 -30>, angularTarget=<1 0 0>}
      localJointOffset=<2 0 0>
     }]
   }
   n16 = Anim Dynamics - Bone: dyn_Leg_Tassle_l_03_jnt!(ComponentPose=@n15, Alpha=1, GravityOverride:Vector="0,0,0", ExternalForce="0,0,0") -> (Pose) {
    angularDampingOverride=1
    angularSpringConstant=200
    bAngularSpring=true
    bOverrideAngularDamping=true
    bUseGravityOverride=true
    simulationSpace=RootRelative
    alphaBoolBlend {blendOption=Linear}
    boundBone {boneName=dyn_Leg_Tassle_l_03_jnt}
    componentAppliedLinearAccClamp=<6000 6000 6000>
    componentLinearAccScale=<0.3 0.3 0.3>
    componentLinearVelScale=<0.3 0.3 0.3>
    physicsBodyDefinitions [{
      boundBone {boneName=dyn_Leg_Tassle_l_03_jnt}
      boxExtents=<10 10 10>
      constraintSetup {angularLimitsMax=<1 6 30>, angularLimitsMin=<-1 -30 -30>, angularTarget=<1 0 0>}
      localJointOffset=<2 0 0>
     }]
   }
   n17 = Anim Dynamics - Bone: dyn_Leg_Tassle_r_01_jnt!(ComponentPose=@n16, Alpha=1, GravityOverride:Vector="0,0,0", ExternalForce="0,0,0") -> (Pose) {
    angularDampingOverride=1
    angularSpringConstant=200
    bAngularSpring=true
    bOverrideAngularDamping=true
    bUseGravityOverride=true
    simulationSpace=RootRelative
    alphaBoolBlend {blendOption=Linear}
    boundBone {boneName=dyn_Leg_Tassle_r_01_jnt}
    componentAppliedLinearAccClamp=<6000 6000 6000>
    componentLinearAccScale=<0.3 0.3 0.3>
    componentLinearVelScale=<0.3 0.3 0.3>
    physicsBodyDefinitions [{
      boundBone {boneName=dyn_Leg_Tassle_r_01_jnt}
      boxExtents=<10 10 10>
      constraintSetup {angularLimitsMax=<1 6 30>, angularLimitsMin=<-1 -30 -30>, angularTarget=<1 0 0>}
      localJointOffset=<2 0 0>
     }]
   }
   n18 = Anim Dynamics - Bone: dyn_Leg_Tassle_r_02_jnt!(ComponentPose=@n17, Alpha=1, GravityOverride:Vector="0,0,0", ExternalForce="0,0,0") -> (Pose) {
    angularDampingOverride=1
    angularSpringConstant=200
    bAngularSpring=true
    bOverrideAngularDamping=true
    bUseGravityOverride=true
    simulationSpace=RootRelative
    alphaBoolBlend {blendOption=Linear}
    boundBone {boneName=dyn_Leg_Tassle_r_02_jnt}
    componentAppliedLinearAccClamp=<6000 6000 6000>
    componentLinearAccScale=<0.3 0.3 0.3>
    componentLinearVelScale=<0.3 0.3 0.3>
    physicsBodyDefinitions [{
      boundBone {boneName=dyn_Leg_Tassle_r_02_jnt}
      boxExtents=<10 10 10>
      constraintSetup {angularLimitsMax=<1 6 30>, angularLimitsMin=<-1 -30 -30>, angularTarget=<1 0 0>}
      localJointOffset=<2 0 0>
     }]
   }
   n19 = Anim Dynamics - Bone: dyn_Leg_Tassle_r_03_jnt!(ComponentPose=@n18, Alpha=1, GravityOverride:Vector="0,0,0", ExternalForce="0,0,0") -> (Pose) {
    angularDampingOverride=1
    angularSpringConstant=200
    bAngularSpring=true
    bOverrideAngularDamping=true
    bUseGravityOverride=true
    simulationSpace=RootRelative
    alphaBoolBlend {blendOption=Linear}
    boundBone {boneName=dyn_Leg_Tassle_r_03_jnt}
    componentAppliedLinearAccClamp=<6000 6000 6000>
    componentLinearAccScale=<0.3 0.3 0.3>
    componentLinearVelScale=<0.3 0.3 0.3>
    physicsBodyDefinitions [{
      boundBone {boneName=dyn_Leg_Tassle_r_03_jnt}
      boxExtents=<10 10 10>
      constraintSetup {angularLimitsMax=<1 6 30>, angularLimitsMin=<-1 -30 -30>, angularTarget=<1 0 0>}
      localJointOffset=<2 0 0>
     }]
   }
   n20 = Anim Dynamics - Bone: dyn_Jacket_ft_l_01_jnt!(ComponentPose=@n19, Alpha=1, GravityOverride:Vector="0,0,0", ExternalForce="0,0,0") -> (Pose) {
    angularDampingOverride=0.9999
    angularSpringConstant=800
    bAngularSpring=true
    bOverrideAngularDamping=true
    bUseGravityOverride=true
    simulationSpace=RootRelative
    alphaBoolBlend {blendOption=Linear}
    boundBone {boneName=dyn_Jacket_ft_l_01_jnt}
    componentAppliedLinearAccClamp=<6000 6000 6000>
    componentLinearAccScale=<0.3 0.3 0.3>
    componentLinearVelScale=<0.3 0.3 0.3>
    physicsBodyDefinitions [{
      boundBone {boneName=dyn_Jacket_ft_l_01_jnt}
      boxExtents=<20 20 20>
      constraintSetup {angularLimitsMax=<1 1 5>, angularLimitsMin=<-1 -30 -5>, angularTarget=<1 0 0>}
      localJointOffset=<5 0 0>
     }]
   }
   n21 = Anim Dynamics - Bone: dyn_Jacket_ft_l_02_jnt!(ComponentPose=@n20, Alpha=1, GravityOverride:Vector="0,0,0", ExternalForce="0,0,0") -> (Pose) {
    angularDampingOverride=0.9999
    angularSpringConstant=800
    bAngularSpring=true
    bOverrideAngularDamping=true
    bUseGravityOverride=true
    simulationSpace=RootRelative
    alphaBoolBlend {blendOption=Linear}
    boundBone {boneName=dyn_Jacket_ft_l_02_jnt}
    componentAppliedLinearAccClamp=<6000 6000 6000>
    componentLinearAccScale=<0.3 0.3 0.3>
    componentLinearVelScale=<0.3 0.3 0.3>
    physicsBodyDefinitions [{
      boundBone {boneName=dyn_Jacket_ft_l_02_jnt}
      boxExtents=<20 20 20>
      constraintSetup {angularLimitsMax=<1 1 5>, angularLimitsMin=<-1 -30 -5>, angularTarget=<1 0 0>}
      localJointOffset=<5 0 0>
     }]
   }
   n22 = Anim Dynamics - Bone: dyn_Zipper_l_01_jnt!(ComponentPose=@n21, Alpha=1, GravityOverride:Vector="0,0,0", ExternalForce="0,0,0") -> (Pose) {
    angularDampingOverride=0.9999
    angularSpringConstant=800
    bAngularSpring=true
    bOverrideAngularDamping=true
    bUseGravityOverride=true
    simulationSpace=RootRelative
    alphaBoolBlend {blendOption=Linear}
    boundBone {boneName=dyn_Zipper_l_01_jnt}
    componentAppliedLinearAccClamp=<6000 6000 6000>
    componentLinearAccScale=<0.3 0.3 0.3>
    componentLinearVelScale=<0.3 0.3 0.3>
    physicsBodyDefinitions [{
      boundBone {boneName=dyn_Zipper_l_01_jnt}
      boxExtents=<20 20 20>
      constraintSetup {angularLimitsMax=<1 1 5>, angularLimitsMin=<-1 -30 -5>, angularTarget=<1 0 0>}
      localJointOffset=<5 0 0>
     }]
   }
   n23 = Anim Dynamics - Bone: dyn_Jacket_l_01_jnt!(ComponentPose=@n22, Alpha=1, GravityOverride:Vector="0,0,0", ExternalForce="0,0,0") -> (Pose) {
    angularDampingOverride=0.9999
    angularSpringConstant=800
    bAngularSpring=true
    bOverrideAngularDamping=true
    bUseGravityOverride=true
    simulationSpace=RootRelative
    alphaBoolBlend {blendOption=Linear}
    boundBone {boneName=dyn_Jacket_l_01_jnt}
    componentAppliedLinearAccClamp=<6000 6000 6000>
    componentLinearAccScale=<0.3 0.3 0.3>
    componentLinearVelScale=<0.3 0.3 0.3>
    physicsBodyDefinitions [{
      boundBone {boneName=dyn_Jacket_l_01_jnt}
      boxExtents=<20 20 20>
      constraintSetup {angularLimitsMax=<1 1 5>, angularLimitsMin=<-1 -30 -5>, angularTarget=<1 0 0>}
      localJointOffset=<5 0 0>
     }]
   }
   n24 = Anim Dynamics - Bone: dyn_Jacket_l_02_jnt!(ComponentPose=@n23, Alpha=1, GravityOverride:Vector="0,0,0", ExternalForce="0,0,0") -> (Pose) {
    angularDampingOverride=0.9999
    angularSpringConstant=800
    bAngularSpring=true
    bOverrideAngularDamping=true
    bUseGravityOverride=true
    simulationSpace=RootRelative
    alphaBoolBlend {blendOption=Linear}
    boundBone {boneName=dyn_Jacket_l_02_jnt}
    componentAppliedLinearAccClamp=<6000 6000 6000>
    componentLinearAccScale=<0.3 0.3 0.3>
    componentLinearVelScale=<0.3 0.3 0.3>
    physicsBodyDefinitions [{
      boundBone {boneName=dyn_Jacket_l_02_jnt}
      boxExtents=<20 20 20>
      constraintSetup {angularLimitsMax=<1 1 5>, angularLimitsMin=<-1 -30 -5>, angularTarget=<1 0 0>}
      localJointOffset=<5 0 0>
     }]
   }
   n25 = Anim Dynamics - Bone: dyn_Jacket_ft_r_01_jnt!(ComponentPose=@n24, Alpha=0.75, GravityOverride:Vector="0,0,0", ExternalForce="0,0,0") -> (Pose) {
    angularDampingOverride=0.9999
    angularSpringConstant=1200
    bAngularSpring=true
    bOverrideAngularDamping=true
    bUseGravityOverride=true
    simulationSpace=RootRelative
    alphaBoolBlend {blendOption=Linear}
    boundBone {boneName=dyn_Jacket_ft_r_01_jnt}
    componentAppliedLinearAccClamp=<6000 6000 6000>
    componentLinearAccScale=<0.3 0.3 0.3>
    componentLinearVelScale=<0.3 0.3 0.3>
    physicsBodyDefinitions [{
      boundBone {boneName=dyn_Jacket_ft_r_01_jnt}
      boxExtents=<20 20 20>
      constraintSetup {angularLimitsMax=<1 1 5>, angularLimitsMin=<-1 -30 -5>, angularTarget=<1 0 0>}
      localJointOffset=<5 0 0>
     }]
   }
   n26 = Anim Dynamics - Bone: dyn_Jacket_ft_r_02_jnt!(ComponentPose=@n25, Alpha=0.75, GravityOverride:Vector="0,0,0", ExternalForce="0,0,0") -> (Pose) {
    angularDampingOverride=0.9999
    angularSpringConstant=1200
    bAngularSpring=true
    bOverrideAngularDamping=true
    bUseGravityOverride=true
    simulationSpace=RootRelative
    alphaBoolBlend {blendOption=Linear}
    boundBone {boneName=dyn_Jacket_ft_r_02_jnt}
    componentAppliedLinearAccClamp=<6000 6000 6000>
    componentLinearAccScale=<0.3 0.3 0.3>
    componentLinearVelScale=<0.3 0.3 0.3>
    physicsBodyDefinitions [{
      boundBone {boneName=dyn_Jacket_ft_r_02_jnt}
      boxExtents=<20 20 20>
      constraintSetup {angularLimitsMax=<1 1 5>, angularLimitsMin=<-1 -30 -5>, angularTarget=<1 0 0>}
      localJointOffset=<5 0 0>
     }]
   }
   n27 = Anim Dynamics - Bone: dyn_Jacket_r_01_jnt!(ComponentPose=@n26, Alpha=0.75, GravityOverride:Vector="0,0,0", ExternalForce="0,0,0") -> (Pose) {
    angularDampingOverride=0.9999
    angularSpringConstant=1200
    bAngularSpring=true
    bOverrideAngularDamping=true
    bUseGravityOverride=true
    simulationSpace=RootRelative
    alphaBoolBlend {blendOption=Linear}
    boundBone {boneName=dyn_Jacket_r_01_jnt}
    componentAppliedLinearAccClamp=<6000 6000 6000>
    componentLinearAccScale=<0.3 0.3 0.3>
    componentLinearVelScale=<0.3 0.3 0.3>
    physicsBodyDefinitions [{
      boundBone {boneName=dyn_Jacket_r_01_jnt}
      boxExtents=<20 20 20>
      constraintSetup {angularLimitsMax=<1 1 5>, angularLimitsMin=<-1 -30 -5>, angularTarget=<1 0 0>}
      localJointOffset=<5 0 0>
     }]
   }
   n28 = Anim Dynamics - Bone: dyn_Jacket_r_02_jnt!(ComponentPose=@n27, Alpha=0.75, GravityOverride:Vector="0,0,0", ExternalForce="0,0,0") -> (Pose) {
    angularDampingOverride=0.9999
    angularSpringConstant=1200
    bAngularSpring=true
    bOverrideAngularDamping=true
    bUseGravityOverride=true
    simulationSpace=RootRelative
    alphaBoolBlend {blendOption=Linear}
    boundBone {boneName=dyn_Jacket_r_02_jnt}
    componentAppliedLinearAccClamp=<6000 6000 6000>
    componentLinearAccScale=<0.3 0.3 0.3>
    componentLinearVelScale=<0.3 0.3 0.3>
    physicsBodyDefinitions [{
      boundBone {boneName=dyn_Jacket_r_02_jnt}
      boxExtents=<20 20 20>
      constraintSetup {angularLimitsMax=<1 1 5>, angularLimitsMin=<-1 -30 -5>, angularTarget=<1 0 0>}
      localJointOffset=<5 0 0>
     }]
   }
   n29 = "Transform (Modify) Bone - Bone: dyn_Wristband_l_01_jnt"(ComponentPose=@n28, Alpha=1, Translation="-2,0,0", Rotation="0,0,0", Scale="1,1,1") {
    translationMode=BMM_Additive
    translationSpace=BCS_BoneSpace
    alphaBoolBlend {blendOption=Linear}
    boneToModify {boneName=dyn_Wristband_l_01_jnt}
    translation {x=-2}
   }
   n30 = Anim Dynamics - Bone: dyn_Wristband_l_01_jnt!(ComponentPose=@n29, Alpha=0.5, GravityOverride:Vector="0,0,0", ExternalForce="0,0,0") -> (Pose) {
    angularDampingOverride=1
    angularSpringConstant=800
    bOverrideAngularDamping=true
    bOverrideLinearDamping=true
    bUseGravityOverride=true
    linearDampingOverride=1
    linearSpringConstant=50
    simulationSpace=RootRelative
    alphaBoolBlend {blendOption=Linear}
    boundBone {boneName=dyn_Wristband_l_01_jnt}
    componentAppliedLinearAccClamp=<6000 6000 6000>
    physicsBodyDefinitions [{
      boundBone {boneName=dyn_Wristband_l_01_jnt}
      boxExtents=<40 10 10>
      constraintSetup {angularLimitsMax=<1 5 1>, angularLimitsMin=<-1 -5 -0.5>, angularTarget=<1 0 0>, linearAxesMax=<0.75 0 0>, linearAxesMin=<-3 0 0>}
      localJointOffset=<1 0 0>
     }]
   }
   n31 = Anim Dynamics - Bone: dyn_WaistPack_01_jnt!(ComponentPose=@n30, Alpha=1, GravityOverride:Vector="0,0,0", ExternalForce="0,0,0") -> (Pose) {
    angularDampingOverride=1
    angularSpringConstant=800
    bOverrideAngularDamping=true
    bUseGravityOverride=true
    linearDampingOverride=1
    linearSpringConstant=50
    simulationSpace=RootRelative
    alphaBoolBlend {blendOption=Linear}
    boundBone {boneName=dyn_WaistPack_01_jnt}
    componentAppliedLinearAccClamp=<6000 6000 6000>
    componentLinearAccScale=<0.3 0.3 0.3>
    componentLinearVelScale=<0.3 0.3 0.3>
    physicsBodyDefinitions [{
      boundBone {boneName=dyn_WaistPack_01_jnt}
      boxExtents=<20 20 20>
      constraintSetup {angularLimitsMax=<1 5 1>, angularLimitsMin=<-10 -15 -1>, angularTarget=<1 0 0>}
      localJointOffset=<3 0 0>
     }]
   }
   n32 = RigidBody!(ComponentPose=@n31, Alpha, OverrideWorldGravity:Vector="0,0,0", ExternalForce) -> (Pose) {
    bClampLinearTranslationLimitToRefPose=true
    bOverrideWorldGravity=true
    overridePhysicsAsset=&/F_MED_SilverBellWind_Physics
    simulationSpace=BaseBoneSpace
    alphaBoolBlend {blendOption=Linear}
    alphaScaleBiasClamp {interpSpeedIncreasing=2}
    baseBoneRef {boneName=pelvis}
    componentAppliedLinearAccClamp=<3000 3000 3000>
    componentLinearAccScale=<0.4 0.3 0.3>
    componentLinearVelScale=<0.4 0.3 0.3>
    simSpaceSettings {
     maxAngularAcceleration=1000
     maxAngularVelocity=3
     maxLinearAcceleration=700
     maxLinearVelocity=500
     velocityScaleZ=0
     worldAlpha=0.25
     externalAngularVelocity {z=0.1}
     externalLinearDragV=<0.2 0.2>
    }
   }
   n33 = Get Is Crouching Alpha()
   n34 = "Transform (Modify) Bone - Bone: dyn_waiststrand_r_MSTR"(ComponentPose=@n32, Alpha=@n33, Translation="0,0,0", Rotation="30,0,0", Scale="1,1,1") {
    rotationMode=BMM_Additive
    rotationSpace=BCS_BoneSpace
    translationMode=BMM_Additive
    translationSpace=BCS_BoneSpace
    alphaBoolBlend {blendOption=Linear}
    boneToModify {boneName=dyn_waiststrand_r_MSTR}
    rotation {pitch=30}
   }
   n35 = "Transform (Modify) Bone - Bone: dyn_waiststrand_l_MSTR"(ComponentPose=@n34, Alpha=@n33, Translation="0,0,0", Rotation="30,0,0", Scale="1,1,1") {
    rotationMode=BMM_Additive
    rotationSpace=BCS_BoneSpace
    translationMode=BMM_Additive
    translationSpace=BCS_BoneSpace
    alphaBoolBlend {blendOption=Linear}
    boneToModify {boneName=dyn_waiststrand_l_MSTR}
    rotation {pitch=30}
   }
   Output Pose(Result=Component To Local(ComponentPose=@n35)) {layerGroup=None}
  ]}
 EventGraph {nodes [
   Event BlueprintUpdateAnimation()
   > n1 = Sequence!(execute) -> (then_0, then_1, then_2:Exec, then_3:Exec, then_4:Exec)
   > n2 = Gravity Override State Function DEPRECATED(CharacterPartAnimInstance=Self-Reference(), GameTimeInSeconds=GetGameTimeInSeconds(), GravityOverrideParams=Get Gravity Override Params())
   > Set Gravity Override(Gravity Override=@n2)
   Set Is Crouching Alpha(execute=@n1.then_1, Is Crouching Alpha:double="To Float (Boolean)"(InBool=Get bIsCrouching())) -> (Output_Get:double)
   n3 = "To Float (Boolean)"(InBool=NOT Boolean(A=Get bIsPlayingEmote()))
   n4 = "To Float (Boolean)"(InBool=NOT Boolean(A=Get bIsOnGrindRail()))
   Set Sleeve Alpha(execute=@n1.then_2, Sleeve Alpha:double=float * float!(A=@n3, B=@n4, C:double=0.8) -> (ReturnValue) {is=title K2Node_PromotableOperator}) -> (Output_Get:double)
   n5 = NOT Boolean(A=Get bIsCrouching())
   Set Shorts Alpha(execute=@n1.then_3, Shorts Alpha:double=float * float(A=@n4, B=@n3, C="To Float (Boolean)"(InBool=@n5))) -> (Output_Get:double)
  ]}
]
variables [
 Gravity Override {type=Vector}
 Gravity Override Params {type=GravityOverrideParamsStruct}
 Is Crouching Alpha {type=number}
 Shorts Alpha {type=number}
 Sleeve Alpha {type=number}
]
